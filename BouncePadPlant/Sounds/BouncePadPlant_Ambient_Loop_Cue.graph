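# Looping ambient sound cue for the BouncePadPlant trap.
# A looping wave player feeds a Modulator node whose pitch range is fixed, so pitch does not vary.
# The cue plays at reduced volume in the Fort_Traps sound class, with a dedicated loop attenuation asset limiting its audible range.
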SoundCue /BouncePadPlant/Sounds/BouncePadPlant_Ambient_Loop_Cue
root=Modulator_0
nodes [
 {id=Modulator_0, children [WavePlayer_8], properties {pitchMax=1, pitchMin=1}}
 {id=WavePlayer_8, properties {bLooping=true, soundWaveAssetPtr=&/BouncePadPlantAmbient_Loop}}
]
properties {
 attenuationSettings=&/Fort_Traps_BouncePadPlant_Loop_Att
 maxDistance=1400
 soundClassObject=&Fort_Traps
 volumeMultiplier=0.3
}
