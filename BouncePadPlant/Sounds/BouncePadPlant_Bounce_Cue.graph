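# Sound cue for the bounce pad plant's bounce, routed to the traps sound class with its own attenuation.
# A Mixer layers a quieter random slap-splash under a louder random bounce, then feeds a root Modulator.
# Each layer branches by quality level: the first two levels pick from six waves, the last two from three-wave subsets.
# The reduced random nodes are excluded from branch culling.
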SoundCue /BouncePadPlant/Sounds/BouncePadPlant_Bounce_Cue
root=Modulator_0
nodes [
 {id=Mixer_0, children [QualityLevel_1, QualityLevel_0], properties {inputVolume [0.5,1.2]}}
 {id=Modulator_0, children [Mixer_0]}
 {id=QualityLevel_0, children [Random_1,Random_1,Random_2,Random_2]}
 {id=QualityLevel_1, children [Random_0,Random_0,Random_3,Random_3]}
 {
  id=Random_0
  children [WavePlayer_6, WavePlayer_7, WavePlayer_14, WavePlayer_15, WavePlayer_16, WavePlayer_17]
  properties {weights [1,1,1,1,1,1]}
 }
 {id=Random_1, children [WavePlayer_0,WavePlayer_1,WavePlayer_2,WavePlayer_3,WavePlayer_4,WavePlayer_5], properties {weights [1,1,1,1,1,1]}}
 {id=Random_2, children [WavePlayer_3,WavePlayer_4,WavePlayer_5], properties {bShouldExcludeFromBranchCulling=true, weights [1,1,1]}}
 {id=Random_3, children [WavePlayer_15, WavePlayer_16, WavePlayer_17], properties {bShouldExcludeFromBranchCulling=true, weights [1,1,1]}}
 {id=WavePlayer_0, properties {soundWaveAssetPtr=&/BouncePadPlant_Bounce_06}}
 {id=WavePlayer_1, properties {soundWaveAssetPtr=&/BouncePadPlant_Bounce_02}}
 {id=WavePlayer_14, properties {soundWaveAssetPtr=&/Bounce_Plant_SlapSplash_03}}
 {id=WavePlayer_15, properties {soundWaveAssetPtr=&/Bounce_Plant_SlapSplash_04}}
 {id=WavePlayer_16, properties {soundWaveAssetPtr=&/Bounce_Plant_SlapSplash_05}}
 {id=WavePlayer_17, properties {soundWaveAssetPtr=&/Bounce_Plant_SlapSplash_06}}
 {id=WavePlayer_2, properties {soundWaveAssetPtr=&/BouncePadPlant_Bounce_03}}
 {id=WavePlayer_3, properties {soundWaveAssetPtr=&/BouncePadPlant_Bounce_04}}
 {id=WavePlayer_4, properties {soundWaveAssetPtr=&/BouncePadPlant_Bounce_05}}
 {id=WavePlayer_5, properties {soundWaveAssetPtr=&/BouncePadPlant_Bounce_01}}
 {id=WavePlayer_6, properties {soundWaveAssetPtr=&/Bounce_Plant_SlapSplash_01}}
 {id=WavePlayer_7, properties {soundWaveAssetPtr=&/Bounce_Plant_SlapSplash_02}}
]
properties {
 attenuationSettings=&/Fort_Traps_BouncePadPlant_Att
 maxDistance=9000
 soundClassObject=&Fort_Traps
}
