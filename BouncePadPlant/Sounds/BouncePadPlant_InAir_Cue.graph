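# Sound cue for the BouncePadPlant trap, named for the in-air phase, that plays a launch whoosh.
# A Random node picks one of four equally weighted whoosh waves and feeds it into a root Modulator.
# Output is boosted in volume, routed to the Fort_Traps sound class, and spatialized with the bounce pad plant's attenuation asset.
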SoundCue /BouncePadPlant/Sounds/BouncePadPlant_InAir_Cue
root=Modulator_0
nodes [
 {id=Modulator_0, children [Random_4]}
 {id=Random_4, children [WavePlayer_8, WavePlayer_9, WavePlayer_10, WavePlayer_11], properties {weights [1,1,1,1]}}
 {id=WavePlayer_10, properties {soundWaveAssetPtr=&/BouncePadPlant_LaunchWhoosh_03}}
 {id=WavePlayer_11, properties {soundWaveAssetPtr=&/BouncePadPlant_LaunchWhoosh_04}}
 {id=WavePlayer_8, properties {soundWaveAssetPtr=&/BouncePadPlant_LaunchWhoosh_01}}
 {id=WavePlayer_9, properties {soundWaveAssetPtr=&/BouncePadPlant_LaunchWhoosh_02}}
]
properties {
 attenuationSettings=&/Fort_Traps_BouncePadPlant_Att
 maxDistance=9000
 soundClassObject=&Fort_Traps
 volumeMultiplier=1.5
}
