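# Sound cue for the melee impact on the Bounce Pad Plant, using the traps sound class and attenuation.
# A modulator feeds a quality-level switch, whose first two levels randomly pick from seven impact waves and last two from a three-wave subset.
# All variations are equally weighted, and the smaller random node is excluded from branch culling.
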SoundCue /BouncePadPlant/Sounds/BouncePadPlant_Melee_Impact_Cue
root=Modulator_0
nodes [
 {id=Modulator_0, children [QualityLevel_1]}
 {id=QualityLevel_1, children [Random_0,Random_0,Random_3,Random_3]}
 {
  id=Random_0
  children [WavePlayer_9, WavePlayer_10, WavePlayer_11, WavePlayer_12, WavePlayer_13, WavePlayer_18, WavePlayer_19]
  properties {weights [1,1,1,1,1,1,1]}
 }
 {id=Random_3, children [WavePlayer_13, WavePlayer_18, WavePlayer_19], properties {bShouldExcludeFromBranchCulling=true, weights [1,1,1]}}
 {id=WavePlayer_10, properties {soundWaveAssetPtr=&/Bounce_Plant_Impact_Melee_03}}
 {id=WavePlayer_11, properties {soundWaveAssetPtr=&/Bounce_Plant_Impact_Melee_04}}
 {id=WavePlayer_12, properties {soundWaveAssetPtr=&/Bounce_Plant_Impact_Melee_05}}
 {id=WavePlayer_13, properties {soundWaveAssetPtr=&/Bounce_Plant_Impact_Melee_06}}
 {id=WavePlayer_18, properties {soundWaveAssetPtr=&/Bounce_Plant_Impact_Melee_07}}
 {id=WavePlayer_19, properties {soundWaveAssetPtr=&/Bounce_Plant_Impact_Melee_08}}
 {id=WavePlayer_9, properties {soundWaveAssetPtr=&/Bounce_Plant_Impact_Melee_02}}
]
properties {
 attenuationSettings=&Fort_Traps_Att
 maxDistance=6000
 soundClassObject=&Fort_Traps
}
