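# Sound cue for the bounce pad plant's projectile impact, playing one of eight recorded variations.
# A quality-level node feeds a root modulator, selecting an equally weighted random pool of all eight waves for its first two slots.
# The last two slots share a smaller pool reusing the final three waves, which is excluded from branch culling.
# Trap sound class and attenuation settings govern mixing and falloff.
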SoundCue /BouncePadPlant/Sounds/BouncePadPlant_Projectile_Impact_Cue
root=Modulator_0
nodes [
 {id=Modulator_0, children [QualityLevel_1]}
 {id=QualityLevel_1, children [Random_0,Random_0,Random_3,Random_3]}
 {
  id=Random_0
  children [WavePlayer_0,WavePlayer_1,WavePlayer_2,WavePlayer_3,WavePlayer_4,WavePlayer_5,WavePlayer_6,WavePlayer_7]
  properties {weights [1,1,1,1,1,1,1,1]}
 }
 {id=Random_3, children [WavePlayer_5,WavePlayer_6,WavePlayer_7], properties {bShouldExcludeFromBranchCulling=true, weights [1,1,1]}}
 {id=WavePlayer_0, properties {soundWaveAssetPtr=&/Bounce_Plant_Impact_Projectile_01}}
 {id=WavePlayer_1, properties {soundWaveAssetPtr=&/Bounce_Plant_Impact_Projectile_02}}
 {id=WavePlayer_2, properties {soundWaveAssetPtr=&/Bounce_Plant_Impact_Projectile_03}}
 {id=WavePlayer_3, properties {soundWaveAssetPtr=&/Bounce_Plant_Impact_Projectile_04}}
 {id=WavePlayer_4, properties {soundWaveAssetPtr=&/Bounce_Plant_Impact_Projectile_05}}
 {id=WavePlayer_5, properties {soundWaveAssetPtr=&/Bounce_Plant_Impact_Projectile_06}}
 {id=WavePlayer_6, properties {soundWaveAssetPtr=&/Bounce_Plant_Impact_Projectile_07}}
 {id=WavePlayer_7, properties {soundWaveAssetPtr=&/Bounce_Plant_Impact_Projectile_08}}
]
properties {
 attenuationSettings=&Fort_Traps_Att
 maxDistance=6000
 soundClassObject=&Fort_Traps
}
